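# Sound cue that plays the shared Animal_Untamed sound for wildlife, by its name the audio for an animal becoming untamed.
# A single wave player feeds the output directly, with no randomization or modulation.
# It uses the wildlife taming attenuation asset and the Fort_Creatures sound class, with a concurrency override that caps simultaneous instances at two.
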SoundCue /Irwin/Sounds/Shared/Animal_Untamed_Cue
root=WavePlayer_0
nodes [{id=WavePlayer_0, properties {soundWaveAssetPtr=&/Animal_Untamed}}]
properties {
 attenuationSettings=/Irwin/Sounds/Tamed/NPC_Wildlife_Tame_Att
 bOverrideConcurrency=true
 maxDistance=2000
 soundClassObject=/Game/Sounds/FortSoundClassesAndModes/Fort_Creatures
 volumeMultiplier=1
 concurrencyOverrides {maxCount=2}
}
