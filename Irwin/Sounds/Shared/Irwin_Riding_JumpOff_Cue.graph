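# Sound cue for the jump-off sound when riding the Irwin creature, classed under creature sounds with the prey gallop attenuation.
# A Random node picks one of three equally weighted JumpOff waves.
# A Branch on the IsLocal parameter plays that pick directly when true.
# When false or unset, it routes the same Random node through a Modulator that lowers volume to roughly a quarter or less.
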SoundCue /Irwin/Sounds/Shared/Irwin_Riding_JumpOff_Cue
root=Branch_0
nodes [
 {id=Branch_0, children [Random_1,Modulator_0,Modulator_0], properties {boolParameterName=IsLocal}}
 {id=Modulator_0, children [Random_1], properties {pitchMax=1, pitchMin=1, volumeMax=0.25, volumeMin=0.2}}
 {id=Random_1, children [WavePlayer_3,WavePlayer_4,WavePlayer_5], properties {weights [1,1,1]}}
 {id=WavePlayer_3, properties {soundWaveAssetPtr=&/Irwin_Riding_JumpOff_03}}
 {id=WavePlayer_4, properties {soundWaveAssetPtr=&/Irwin_Riding_JumpOff_01}}
 {id=WavePlayer_5, properties {soundWaveAssetPtr=&/Irwin_Riding_JumpOff_02}}
]
properties {
 attenuationSettings=/Irwin/Sounds/Settings/Prey_Gallop_Att
 maxDistance=3200
 soundClassObject=/Game/Sounds/FortSoundClassesAndModes/Fort_Creatures
 volumeMultiplier=0.4
}
