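# Sound cue for the Irwin creature jump-on riding sound, played in the creature sound class with prey gallop attenuation.
# A random node picks one of three equally weighted waves and feeds a modulator with default variation.
# A branch on IsLocal plays that chain directly when true, and through a second modulator that sharply lowers volume when false or unset.
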SoundCue /Irwin/Sounds/Shared/Irwin_Riding_JumpOn_Cue
root=Branch_0
nodes [
 {id=Branch_0, children [Modulator_0,Modulator_1,Modulator_1], properties {boolParameterName=IsLocal}}
 {id=Modulator_0, children [Random_0]}
 {id=Modulator_1, children [Modulator_0], properties {pitchMax=1, pitchMin=1, volumeMax=0.25, volumeMin=0.2}}
 {id=Random_0, children [WavePlayer_0,WavePlayer_1,WavePlayer_2], properties {weights [1,1,1]}}
 {id=WavePlayer_0, properties {soundWaveAssetPtr=&/Irwin_Riding_JumpOn_03}}
 {id=WavePlayer_1, properties {soundWaveAssetPtr=&/Irwin_Riding_JumpOn_01}}
 {id=WavePlayer_2, properties {soundWaveAssetPtr=&/Irwin_Riding_JumpOn_02}}
]
properties {
 attenuationSettings=/Irwin/Sounds/Settings/Prey_Gallop_Att
 maxDistance=3200
 soundClassObject=/Game/Sounds/FortSoundClassesAndModes/Fort_Creatures
 volumeMultiplier=1
}
