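# Stereo looping MetaSound source for riding wind, built from three first-person sprint wind loops.
# Three wave players start together on play and feed a three-channel mixer with uneven gains.
# A second mixer sets the summed level through the MSP_AffiliationMixFloat patch, keyed by the character's audio affiliation.
# The patch's Local value reads the CB_RidingWindDuck modulator, so that bus can duck the wind for the local player.
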MetaSoundSource /Irwin/Sounds/Shared/MSS_RidingWind_Loop
inputs [
 Character.Affiliation {type=Enum:FortAudioAffiliation}
 UE.Source.OnPlay {type=Trigger}
 Wave Asset {type=WaveAsset}
 Wave Asset 1 {type=WaveAsset}
 Wave Asset 2 {type=WaveAsset}
]
interfaces [UE.OutputFormat.Stereo:1.0, UE.Source:1.1]
outputs [
 "UE.OutputFormat.Stereo.Audio:0" {type=Audio}
 "UE.OutputFormat.Stereo.Audio:1" {type=Audio}
]
pages [Default {
  input_defaults {
   Character.Affiliation=0
   UE.Source.OnPlay=null
   Wave Asset=&/Riding_1P_Sprint_Wind_Loop
   Wave Asset 1=&/Riding_1P_Sprint_Wind_Loop_02
   Wave Asset 2=&/Riding_1P_Sprint_Wind_Loop_03
  }
  nodes [
   WavePlayer_0 = UE.Wave Player:Stereo(Play=^UE.Source.OnPlay, Wave Asset=^Wave Asset 2) {also {defaults {Loop="<MetaSoundBuilderResult.SUCCEEDED: 0>"}, version [1,1]}}
   WavePlayer_1 = UE.Wave Player:Stereo(Play=^UE.Source.OnPlay, Wave Asset=^Wave Asset 1) {also {defaults {Loop="<MetaSoundBuilderResult.SUCCEEDED: 0>"}, version [1,1]}}
   WavePlayer_2 = UE.Wave Player:Stereo(Play=^UE.Source.OnPlay, Wave Asset=^Wave Asset) {also {defaults {Loop="<MetaSoundBuilderResult.SUCCEEDED: 0>"}, version [1,1]}}
   AudioMixerAudioMixerStereo3_0 = "AudioMixer.Audio Mixer (Stereo, 3)"(
    In 2 L=@WavePlayer_0.Out Left
    In 2 R=@WavePlayer_0.Out Right
    In 1 L=@WavePlayer_1.Out Left
    In 1 R=@WavePlayer_1.Out Right
    In 0 L=@WavePlayer_2.Out Left
    In 0 R=@WavePlayer_2.Out Right
   ) {also {defaults {Gain 0=2, Gain 1=0.2}}}
   AudioMixerAudioMixerStereo2_0 = "AudioMixer.Audio Mixer (Stereo, 2)"(
    In 0 L=@AudioMixerAudioMixerStereo3_0.Out L
    In 0 R=@AudioMixerAudioMixerStereo3_0.Out R
    Gain 0=&MSP_AffiliationMixFloat(
     Affiliation=^Character.Affiliation
     Local=Modulation.GetModulatorValue() {also {defaults {Modulator=&/CB_RidingWindDuck}}}.Out
    ).Out
   ) {also {defaults {Gain 1=0}}}
  ]
  links [
   {dst="@UE.OutputFormat.Stereo.Audio:0", src=AudioMixerAudioMixerStereo2_0.Out L}
   {dst="@UE.OutputFormat.Stereo.Audio:1", src=AudioMixerAudioMixerStereo2_0.Out R}
  ]
 }]
properties {
 assetImportData=&self:AssetImportData
 bLooping=true
 duration=10000
 numChannels=2
 outputFormat=Stereo
 sampleRate=48000
}
