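# Looping wind sound cue for first-person sprinting while riding, routed to the player foley sound class with basic foley attenuation.
# An Enveloper fades the looping riding-wind wave in over about a second and a half.
# A Branch on the IsLocal parameter plays that chain directly when true.
# When false or unset, the same Enveloper passes through a Modulator that lowers its volume to a fixed fraction.
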SoundCue /Irwin/Sounds/Shared/Riding_1P_Sprint_Wind_Loop_Cue
root=Branch_0
nodes [
 {id=Branch_0, children [Enveloper_0,Modulator_0,Modulator_0], properties {boolParameterName=IsLocal}}
 {
  id=Enveloper_0
  children [WavePlayer_1]
  properties {pitchCurve {editorCurveData {keys [{time=0}]}}, volumeCurve {editorCurveData {keys [<at 0 0.0039>,<at 0 0>,<at 0.8033 0.3735>,<at 1.5 1>]}}}
 }
 {id=Modulator_0, children [Enveloper_0], properties {pitchMax=1, pitchMin=1, volumeMax=0.3, volumeMin=0.3}}
 {id=WavePlayer_1, properties {bLooping=true, soundWaveAssetPtr=&/MSS_RidingWind_Loop}}
]
properties {
 attenuationSettings=&Foley_Attenuation_Basic_Att
 maxDistance=1536
 soundClassObject=&Fort_Player_Foley_1P_Athena
 volumeMultiplier=1
}
