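# Sound cue for generic riding impact sounds, classed under Fort_Impacts.
# A Random node picks one of four impact wave variations with equal weight.
# A Modulator at the root applies a slight upward pitch variation to each play.
# The cue uses the shared prop impact attenuation and a modest volume boost.
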SoundCue /Irwin/Sounds/Shared/Riding_Impact_Generic_Cue
root=Modulator_0
nodes [
 {id=Modulator_0, children [Random_0], properties {pitchMax=1.1, pitchMin=1}}
 {id=Random_0, children [WavePlayer_0,WavePlayer_1,WavePlayer_2,WavePlayer_3], properties {weights [1,1,1,1]}}
 {id=WavePlayer_0, properties {soundWaveAssetPtr=&/Riding_Impact_Generic_03}}
 {id=WavePlayer_1, properties {soundWaveAssetPtr=&/Riding_Impact_Generic_04}}
 {id=WavePlayer_2, properties {soundWaveAssetPtr=&/Riding_Impact_Generic_01}}
 {id=WavePlayer_3, properties {soundWaveAssetPtr=&/Riding_Impact_Generic_02}}
]
properties {
 attenuationSettings=/Game/Sounds/Attenuation/Fort_Prop_Impact_Att
 maxDistance=3256
 soundClassObject=&Fort_Impacts
 volumeMultiplier=1.2
}
