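# Looping slide sound cue for fauna, built from Athena character dirt-sliding waves and routed to the creature-steps sound class with predator step attenuation.
# Two looping layers, a random dirt bed and random debris, pass through Envelopers that fade volume in over 0.3 seconds, then mix together.
# A random non-looping "Stop" wave, with slight pitch and volume modulation, is mixed over them, so it fires when the cue starts.
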SoundCue /Irwin/Sounds/Shared/Sliding_Fauna_Loop_Cue
root=Mixer_2
nodes [
 {
  id=Enveloper_3
  children [Random_2]
  properties {pitchCurve {editorCurveData {keys [{time=0}]}}, volumeCurve {editorCurveData {keys [<at 0 0>,<at 0 0>,<at 0.3 1>]}}}
 }
 {
  id=Enveloper_4
  children [Random_3]
  properties {pitchCurve {editorCurveData {keys [{time=0}]}}, volumeCurve {editorCurveData {keys [<at 0 0>,<at 0 0>,<at 0.3 1>]}}}
 }
 {id=Mixer_0, children [Enveloper_3,Enveloper_4], properties {inputVolume [1.2,1]}}
 {id=Mixer_2, children [Modulator_1,Mixer_0], properties {inputVolume [1.5,1.2]}}
 {id=Modulator_1, children [Random_5], properties {pitchMax=1.1, pitchMin=1, volumeMax=1.6, volumeMin=1.5}}
 {id=Random_2, children [WavePlayer_18, WavePlayer_19, WavePlayer_20], properties {weights [1,1,1]}}
 {id=Random_3, children [WavePlayer_14, WavePlayer_15, WavePlayer_16, WavePlayer_17], properties {weights [1,1,1,1]}}
 {
  id=Random_5
  children [WavePlayer_22, WavePlayer_23, WavePlayer_24, WavePlayer_25, WavePlayer_26, WavePlayer_27]
  properties {weights [1,1,1,1,1,1]}
 }
 {id=WavePlayer_14, properties {bLooping=true, soundWaveAssetPtr=/Game/Athena/Sounds/Character/Sliding/Waves/Dirt/Sliding_Dirt_Debris_01}}
 {id=WavePlayer_15, properties {bLooping=true, soundWaveAssetPtr=/Game/Athena/Sounds/Character/Sliding/Waves/Dirt/Sliding_Dirt_Debris_02}}
 {id=WavePlayer_16, properties {bLooping=true, soundWaveAssetPtr=/Game/Athena/Sounds/Character/Sliding/Waves/Dirt/Sliding_Dirt_Debris_03}}
 {id=WavePlayer_17, properties {bLooping=true, soundWaveAssetPtr=/Game/Athena/Sounds/Character/Sliding/Waves/Dirt/Sliding_Dirt_Debris_04}}
 {id=WavePlayer_18, properties {bLooping=true, soundWaveAssetPtr=/Game/Athena/Sounds/Character/Sliding/Waves/Dirt/Sliding_Dirt_Layer01_01}}
 {id=WavePlayer_19, properties {bLooping=true, soundWaveAssetPtr=/Game/Athena/Sounds/Character/Sliding/Waves/Dirt/Sliding_Dirt_Layer01_02}}
 {id=WavePlayer_20, properties {bLooping=true, soundWaveAssetPtr=/Game/Athena/Sounds/Character/Sliding/Waves/Dirt/Sliding_Dirt_Layer01_03}}
 {id=WavePlayer_22, properties {soundWaveAssetPtr=/Game/Athena/Sounds/Character/Sliding/Waves/Dirt/Sliding_Dirt_Stop_06}}
 {id=WavePlayer_23, properties {soundWaveAssetPtr=/Game/Athena/Sounds/Character/Sliding/Waves/Dirt/Sliding_Dirt_Stop_01}}
 {id=WavePlayer_24, properties {soundWaveAssetPtr=/Game/Athena/Sounds/Character/Sliding/Waves/Dirt/Sliding_Dirt_Stop_02}}
 {id=WavePlayer_25, properties {soundWaveAssetPtr=/Game/Athena/Sounds/Character/Sliding/Waves/Dirt/Sliding_Dirt_Stop_03}}
 {id=WavePlayer_26, properties {soundWaveAssetPtr=/Game/Athena/Sounds/Character/Sliding/Waves/Dirt/Sliding_Dirt_Stop_04}}
 {id=WavePlayer_27, properties {soundWaveAssetPtr=/Game/Athena/Sounds/Character/Sliding/Waves/Dirt/Sliding_Dirt_Stop_05}}
]
properties {
 attenuationSettings=/Irwin/Sounds/Settings/Predator_Step_Default_Att
 maxDistance=4100
 soundClassObject=/Game/Sounds/FortSoundClassesAndModes/Fort_Creatures_Steps_Normal
 volumeMultiplier=1
}
