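# Sound cue for large water fauna swim jumps, reusing the player's first-person swim jump water sounds.
# A random node picks one of four wave variations with equal weight, feeding a modulator that randomly shifts pitch slightly down or up.
# It plays through the player swimming attenuation and water interaction sound class, with boosted volume and an overridden concurrency limit.
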SoundCue /Irwin/Sounds/Shared/Water_Fauna_Large_Swim_Jump_Cue
root=Modulator_0
nodes [
 {id=Modulator_0, children [Random_0], properties {pitchMax=1.3, pitchMin=0.9}}
 {id=Random_0, children [WavePlayer_0,WavePlayer_1,WavePlayer_5,WavePlayer_6], properties {weights [1,1,1,1]}}
 {id=WavePlayer_0, properties {soundWaveAssetPtr=/Game/Athena/Sounds/Water/Interaction/Water_Player_SwimJump_1P_01}}
 {id=WavePlayer_1, properties {soundWaveAssetPtr=/Game/Athena/Sounds/Water/Interaction/Water_Player_SwimJump_1P_02}}
 {id=WavePlayer_5, properties {soundWaveAssetPtr=/Game/Athena/Sounds/Water/Interaction/Water_Player_SwimJump_1P_03}}
 {id=WavePlayer_6, properties {soundWaveAssetPtr=/Game/Athena/Sounds/Water/Interaction/Water_Player_SwimJump_1P_04}}
]
properties {
 attenuationSettings=/Game/Sounds/Attenuation/Fort_Player_Swimming_Athena_Att
 bOverrideConcurrency=true
 maxDistance=4500
 soundClassObject=/Game/Sounds/FortSoundClassesAndModes/Fort_Water_Interaction
 volumeMultiplier=1.5
 concurrencyOverrides {maxCount=9}
}
