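# Sound cue for the swim sprint of large water fauna, reusing third-person player swim-sprint waves from Athena.
# A quality-level node sends the first two levels to an equal-weight random pool of eleven waves and the last two to a four-wave subset.
# A modulator at the root varies each play.
# The cue uses the player swimming attenuation and water-interaction sound class, with boosted volume and its own concurrency limit.
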SoundCue /Irwin/Sounds/Shared/Water_Fauna_Large_Swim_Sprint_Cue
root=Modulator_1
nodes [
 {id=Modulator_1, children [QualityLevel_0]}
 {id=QualityLevel_0, children [Random_0,Random_0,Random_1,Random_1]}
 {
  id=Random_0
  children [
   WavePlayer_1
   WavePlayer_2
   WavePlayer_3
   WavePlayer_4
   WavePlayer_5
   WavePlayer_6
   WavePlayer_7
   WavePlayer_8
   WavePlayer_9
   WavePlayer_10
   WavePlayer_21
  ]
  properties {weights [1,1,1,1,1,1,1,1,1,1,1]}
 }
 {id=Random_1, children [WavePlayer_0, WavePlayer_11, WavePlayer_12, WavePlayer_13], properties {weights [1,1,1,1]}}
 {id=WavePlayer_0, properties {soundWaveAssetPtr=/Game/Athena/Sounds/Water/Interaction/Water_Player_Swim_Sprint_3P_10}}
 {id=WavePlayer_1, properties {soundWaveAssetPtr=&Water_Player_Swim_Sprint_3P_10}}
 {id=WavePlayer_10, properties {soundWaveAssetPtr=/Game/Athena/Sounds/Water/Interaction/Water_Player_Swim_Sprint_3P_08}}
 {id=WavePlayer_11, properties {soundWaveAssetPtr=/Game/Athena/Sounds/Water/Interaction/Water_Player_Swim_Sprint_3P_11}}
 {id=WavePlayer_12, properties {soundWaveAssetPtr=/Game/Athena/Sounds/Water/Interaction/Water_Player_Swim_Sprint_3P_01}}
 {id=WavePlayer_13, properties {soundWaveAssetPtr=/Game/Athena/Sounds/Water/Interaction/Water_Player_Swim_Sprint_3P_02}}
 {id=WavePlayer_2, properties {soundWaveAssetPtr=&Water_Player_Swim_Sprint_3P_11}}
 {id=WavePlayer_21, properties {soundWaveAssetPtr=/Game/Athena/Sounds/Water/Interaction/Water_Player_Swim_Sprint_3P_09}}
 {id=WavePlayer_3, properties {soundWaveAssetPtr=&Water_Player_Swim_Sprint_3P_01}}
 {id=WavePlayer_4, properties {soundWaveAssetPtr=&Water_Player_Swim_Sprint_3P_02}}
 {id=WavePlayer_5, properties {soundWaveAssetPtr=/Game/Athena/Sounds/Water/Interaction/Water_Player_Swim_Sprint_3P_03}}
 {id=WavePlayer_6, properties {soundWaveAssetPtr=/Game/Athena/Sounds/Water/Interaction/Water_Player_Swim_Sprint_3P_04}}
 {id=WavePlayer_7, properties {soundWaveAssetPtr=/Game/Athena/Sounds/Water/Interaction/Water_Player_Swim_Sprint_3P_05}}
 {id=WavePlayer_8, properties {soundWaveAssetPtr=/Game/Athena/Sounds/Water/Interaction/Water_Player_Swim_Sprint_3P_06}}
 {id=WavePlayer_9, properties {soundWaveAssetPtr=/Game/Athena/Sounds/Water/Interaction/Water_Player_Swim_Sprint_3P_07}}
]
properties {
 attenuationSettings=/Game/Sounds/Attenuation/Fort_Player_Swimming_Athena_Att
 bOverrideConcurrency=true
 maxDistance=4500
 soundClassObject=/Game/Sounds/FortSoundClassesAndModes/Fort_Water_Interaction
 volumeMultiplier=1.75
 concurrencyOverrides {maxCount=9}
}
